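# Unlit translucent Niagara sprite and mesh material for the Ranked Progression Electric Emote effect, drawing screen-space halftone dots tinted by particle color.
# The dot grid is aspect-corrected, anchored to the object's clip-space position and scaled by view depth.
# A static switch sizes dots by a quantized radial sphere mask, or by a constant blended toward a pixelated, thresholded texture cutout.
# Near-camera fade and particle alpha drive opacity; a trailing UV-based dot branch appears unconnected.
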Material /FNGameplayCues/Effects/Fort_Effects/Effects/Characters/Generic/RankedProgression/Electric_Emote/M_Electric_Emote_CameraSpace
graph {
 properties {
  bAutomaticallySetUsageInEditor=false
  bUsedWithNiagaraMeshParticles=true
  bUsedWithNiagaraSprites=true
  blendMode=BLEND_Translucent
  pixelDepthOffsetMode=PDOM_Legacy
  shadingModel=MSM_Unlit
  usedShadingModels=Unlit
 }
 nodes [
  n1 = ParticleColor()
  n2 = ConstantBiasScale(None=ScreenPosition().ViewportUV) {bias=-0.5, scale=2}
  n3 = MaterialFunctionCall(Float2=ViewSize()) {materialFunction=&BreakOutFloat2Components}
  n4 = AppendVector(A=Constant() {r=1}, B=Divide(A=@n3.G, B=@n3.G))
  n5 = Multiply(
   A=MaterialFunctionCall(World Position=ObjectPositionWS()) {materialFunction=&TransformToClipSpace}.Clip Space XY
   B=Constant2Vector() {g=-1, r=1}
  )
  n6 = Subtract(A=ConstantBiasScale(None=Multiply(A=@n2, B=@n4)), B=Multiply(A=Multiply(A=@n4, B=@n5)) {constB=0.5})
  n7 = TransformPosition(None=ObjectPositionWS()) {transformSourceType=TRANSFORMPOSSOURCE_World, transformType=TRANSFORMPOSSOURCE_View}
  n8 = Multiply(A=Divide(B=ComponentMask(None=@n7) {b=true}))
  n9 = Frac(None=Multiply(
    A=MaterialFunctionCall(UVs=@n6, Texture Scale=@n8) {materialFunction=&ScaleUVsByCenter}.UVs
    B=ScalarParameter() {defaultValue=1, parameterName=Tiling}
   ))
  n10 = SmoothStep(Value=Length(None=Subtract(A=@n9) {constB=0.5}) {is=class MaterialExpressionLength}) {constMax=0.5}
  n11 = ScalarParameter() {defaultValue=0.4, parameterName=DotScaleMaster}
  n12 = MaterialFunctionCall!(Position=ObjectPositionWS(), UVRadialScale=ScalarParameter() {defaultValue=0.25, parameterName=SphereMaskRadius}) {materialFunction=/Game/Weapons/WeaponSkins/MaterialFunction/ObjectStabilizedScreenSpaceUVs}
  n13 = ScalarParameter() {defaultValue=7, parameterName=SphereMaskResolution}
  n14 = Divide(A=Floor(None=Multiply(A=@n12.StabilizedUVs, B=@n13) {constB=20}), B=@n13) {constB=20}
  n15 = LinearInterpolate(
   A=ScalarParameter() {parameterName=SphereMaskDotScaleRange_Min}
   B=ScalarParameter() {defaultValue=0.6, parameterName=SphereMaskDotScaleRange_Max}
   Alpha=Power(Base=MaterialFunctionCall!(UVs=@n14, CenterPosition, Radius) {materialFunction=/Engine/Functions/Engine_MaterialFunctions01/Gradient/SphereGradient-2D}.Result 0-1)
  ) {constB=0.6}
  n16 = ScalarParameter() {defaultValue=0.5, parameterName=DotSize}
  n17 = ScalarParameter() {parameterName=UseTextureCutout?}
  n18 = LinearInterpolate(
   A=ScalarParameter() {defaultValue=32, parameterName=TextureResolution}
   B=DynamicParameter() {defaultValue=<rgba 32 1 1 1>, paramNames [Texture Resolution, Param2, Param3, Param4]}.Texture Resolution
   Alpha=ScalarParameter() {parameterName=UseDynamicParam?}
  )
  n19 = Multiply(A=TextureCoordinate(), B=@n18) {constB=10}
  n20 = TextureSampleParameter2D(UVs=Divide(A=Floor(None=@n19), B=@n18) {constB=10}) {
   parameterName=TextureInput
   samplerType=SAMPLERTYPE_LinearGrayscale
   texture=/Game/Effects/Fort_Effects/Emotes/QuarrelFemale/T_Emote_Quarrel_Female_Kick_Flash_01
  }
  n21 = LinearInterpolate(Alpha=Step(X=ChannelMaskParameter(None=@n20.RGBA) {parameterName=TextureChannelSelect}.R) {constY=0.8}) {constB=0.5}
  n22 = StaticSwitchParameter(
   True=Step(Y=@n10, X=Add(A=@n11, B=@n15) {constA=0.2})
   False=Step(Y=@n10, X=LinearInterpolate(A=@n16, B=@n21, Alpha=@n17) {constA=0.5}) {constX=0.5}
  ) {parameterName=UseBen-DaySphereMask?}
  n23 = ScalarParameter() {defaultValue=150, parameterName=CameraFadeDistance}
  Reroute(None=Multiply(
    A=Multiply(A=@n22, B=@n1.A)
    B=MaterialFunctionCall(Distance=@n23) {materialFunction=&Near_Camera_Fade}.Result
   ))
  Reroute(None=MaterialFunctionCall(Emissive=Multiply(A=@n22, B=@n1.RGB)) {materialFunction=&MF_FX_EmissiveAdjustments}.Result)
  n24 = TextureCoordinate()
  n25 = ScalarParameter() {defaultValue=33, parameterName=s}
  n26 = Frac(None=Multiply(A=@n24, B=@n25) {constB=20})
  n27 = Floor(None=Multiply(A=@n24, B=@n25) {constB=20})
  n28 = Power(Base=MaterialFunctionCall(UVs=Divide(A=@n27, B=@n25) {constB=20}, Radius=Constant() {r=0.4}) {materialFunction=&SphereGradient-2D}.Result 0-1)
  Step(Y=Length(None=Subtract(A=@n26) {constB=0.5}), X=Add(B=LinearInterpolate(Alpha=@n28) {constB=0.5}) {constA=0.2})
 ]
}
